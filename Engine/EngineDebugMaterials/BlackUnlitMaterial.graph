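# BlackUnlitMaterial is an engine debug material that renders surfaces as solid black, unaffected by lighting.
# It uses the Unlit shading model with an empty expression graph, so emissive color stays at its black default.
# Usage flags enable it for skeletal meshes, mesh deformers and static lighting.
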Material /Engine/EngineDebugMaterials/BlackUnlitMaterial
graph {
 expressions {}
 properties {
  bUsedWithMeshDeformer=true
  bUsedWithSkeletalMesh=true
  bUsedWithStaticLighting=true
  pixelDepthOffsetMode=PDOM_Legacy
  shadingModel=MSM_Unlit
  usedShadingModels=Unlit
 }
}
