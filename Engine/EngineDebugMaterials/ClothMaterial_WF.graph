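# Translucent, unlit, two-sided wireframe debug material in the engine, named for displaying cloth.
# ClothOpacity sets opacity, and BackfaceCull blends it with a version zeroed wherever a transformed Z vector dotted with the pixel normal is not positive.
# Vertex color runs through an sRGB-to-linear curve, then is inverted, biased and scaled down.
# A separate branch transforms a tiny constant vector.
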Material /Engine/EngineDebugMaterials/ClothMaterial_WF
graph {
 properties {
  bUseTranslucencyVertexFog=false
  blendMode=BLEND_Translucent
  pixelDepthOffsetMode=PDOM_Legacy
  shadingModel=MSM_Unlit
  twoSided=true
  usedShadingModels=Unlit
  wireframe=true
 }
 nodes [
  n1 = ScalarParameter() {defaultValue=0.8, parameterName=ClothOpacity, sliderMax=1}
  n2 = ScalarParameter() {defaultValue=1, parameterName=BackfaceCull, sliderMax=1}
  n3 = DotProduct(A=Transform(None=Constant3Vector() {constant=<rgba 0 0 1 0>}), B=PixelNormalWS())
  n4 = Constant()
  n5 = Constant()
  Add(A=Multiply(A=@n1, B=OneMinus(None=@n2)), B=Multiply(A=@n2, B=If(A=@n3, B=@n4, "A > B"=@n1, "A == B"=@n5, "A < B"=@n5)))
  n6 = Divide(A=Add(A=VertexColor()) {constB=0.055}) {constB=1.055}
  ConstantBiasScale(None=OneMinus(None=Power(Base=@n6) {constExponent=2.4})) {bias=0.4, scale=0.2}
  Transform(None=Constant3Vector() {constant=<rgba 0.01 0.01 0.006 0>})
 ]
}
