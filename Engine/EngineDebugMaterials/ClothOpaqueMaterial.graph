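# Engine debug material for cloth, an unlit, two-sided surface built around mesh vertex color.
# Vertex color runs through an add, divide and power chain that applies the curved segment of the sRGB-to-linear conversion.
# A separate branch transforms a tiny constant Z vector, likely a small surface offset, though the file shows no material output connections for either branch.
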Material /Engine/EngineDebugMaterials/ClothOpaqueMaterial
graph {
 properties {bUseTranslucencyVertexFog=false, pixelDepthOffsetMode=PDOM_Legacy, shadingModel=MSM_Unlit, twoSided=true, usedShadingModels=Unlit}
 nodes [
  n1 = VertexColor()
  Power(Base=Divide(A=Add(A=@n1) {constB=0.055}) {constB=1.055}) {constExponent=2.4}
  Transform(None=Constant3Vector() {constant=<rgba 0 0 0.005 0>})
 ]
}
