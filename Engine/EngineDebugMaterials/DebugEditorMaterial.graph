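# Unlit, two-sided engine debug material with depth testing disabled, named for drawing debug visuals in the editor.
# A Color vector parameter, pale yellow-green by default, feeds a Desaturation node whose fraction comes from a Desaturation scalar parameter.
# A separate Opacity scalar parameter defaults to a low value, so color, saturation and opacity are all exposed to material instances.
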Material /Engine/EngineDebugMaterials/DebugEditorMaterial
graph {
 properties {bDisableDepthTest=true, pixelDepthOffsetMode=PDOM_Legacy, shadingModel=MSM_Unlit, twoSided=true, usedShadingModels=Unlit}
 nodes [
  Desaturation(
   None=VectorParameter() {parameterName=Color, defaultValue=<rgba 0.8984 1 0.473 1>}.RGB
   Fraction=ScalarParameter() {parameterName=Desaturation}
  )
  ScalarParameter() {defaultValue=0.3, parameterName=Opacity, sliderMax=1}
 ]
}
