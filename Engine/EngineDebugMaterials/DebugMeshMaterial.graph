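# Engine debug material for meshes, rendered unlit, translucent and two-sided.
# Its only node is a Color vector parameter, white by default, which lets material instances supply a tint.
# The graph lists no link from that parameter to any material output.
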Material /Engine/EngineDebugMaterials/DebugMeshMaterial
graph {
 properties {blendMode=BLEND_Translucent, pixelDepthOffsetMode=PDOM_Legacy, shadingModel=MSM_Unlit, twoSided=true, usedShadingModels=Unlit}
 nodes [MaterialExpressionVectorParameter0 = VectorParameter() {parameterName=Color, defaultValue=<rgba 1 1 1 1>}]
}
