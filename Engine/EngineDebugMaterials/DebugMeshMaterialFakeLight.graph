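# Engine debug mesh material that shades two-sided geometry with a fake directional light, tinted by the GizmoColor parameter.
# It derives a faceted normal from the cross product of screen-space world position derivatives, with an ES2 branch feeding raw position instead.
# The clamped dot product against a fixed diagonal direction drives a diffuse ramp and a powered highlight, which are summed and multiplied by the color.
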Material /Engine/EngineDebugMaterials/DebugMeshMaterialFakeLight
graph {
 properties {pixelDepthOffsetMode=PDOM_Legacy, twoSided=true, usedShadingModels=Default Lit}
 nodes [
  n1 = Constant3Vector() {constant=<rgba -1 -1 -1 0>}
  n2 = WorldPosition()
  n3 = DDX(Value=@n2.XYZ)
  n4 = FeatureLevelSwitch(Default=@n3, ES2=@n2.XYZ, SM4_REMOVED=@n3)
  n5 = DDY(Value=@n2.XYZ)
  n6 = FeatureLevelSwitch(Default=@n5, ES2=@n2.XYZ, SM4_REMOVED=@n5)
  n7 = Clamp(None=DotProduct(A=Normalize(VectorInput=@n1), B=CrossProduct(A=@n4, B=@n6)))
  Add(A=@n7) {constB=0.2}
  Constant3Vector()
  n8 = Power(Base=@n7) {constExponent=20}
  Multiply(
   A=Add(A=Multiply(A=@n8) {constB=0.5}, B=LinearInterpolate(Alpha=@n7) {constA=0.5})
   B=VectorParameter() {parameterName=GizmoColor, defaultValue=<rgba 1 1 1 1>}.RGB
  )
 ]
}
