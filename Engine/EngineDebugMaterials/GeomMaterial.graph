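# GeomMaterial is an unlit translucent engine debug material with depth testing disabled, so it draws over scene geometry.
# An If node compares pixel depth, less a constant offset, against a small threshold.
# Nearer pixels get the full Color vector parameter, while farther pixels get the same color darkened by multiplying it with that threshold constant.
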Material /Engine/EngineDebugMaterials/GeomMaterial
graph {
 properties {
  bDisableDepthTest=true
  blendMode=BLEND_Translucent
  pixelDepthOffsetMode=PDOM_Legacy
  shadingModel=MSM_Unlit
  translucencyPass=MTP_BeforeDOF
  usedShadingModels=Unlit
 }
 nodes [
  n1 = Constant() {r=0.2}
  MaterialExpressionVectorParameter0 = VectorParameter() {parameterName=Color, defaultValue=<rgba 1 1 1 1>}
  If(
   A=Subtract(A=PixelDepth().R, B=Constant() {r=1})
   B=@n1
   "A > B"=Multiply(A=@n1, B=@MaterialExpressionVectorParameter0.RGB)
   "A < B"=@MaterialExpressionVectorParameter0.RGB
  )
 ]
}
