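# Translucent, unlit engine debug material that maps values from a heatmap texture to gradient colours.
# A HeatmapTexture parameter, defaulting to a white square, is sampled with UVs given a fixed quarter turn by a Rotator with constant time.
# Its value is appended with a constant to form coordinates into the HeatmapGradient lookup texture.
# A square root of the same sample is also computed; the graph shows no material output connections.
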Material /Engine/EngineDebugMaterials/HeatmapMaterial
graph {
 properties {bUsedWithStaticLighting=true, blendMode=BLEND_Translucent, pixelDepthOffsetMode=PDOM_Legacy, shadingModel=MSM_Unlit, usedShadingModels=Unlit}
 nodes [
  n1 = TextureSampleParameter2D(UVs=Rotator(Coordinate=TextureCoordinate(), Time=Constant() {r=1.5708}) {speed=1}) {parameterName=HeatmapTexture, texture=/Engine/EngineResources/WhiteSquareTexture}
  SquareRoot(None=@n1.RGB)
  TextureSample(UVs=AppendVector(A=@n1.RGB, B=Constant())) {texture=&/HeatmapGradient}
 ]
}
