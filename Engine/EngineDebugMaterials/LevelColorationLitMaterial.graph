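# Engine debug material that, going by its name, shades geometry in a flat color for the lit level coloration view.
# A Color vector parameter, mid grey by default, feeds the DiffuseColor input of the ConvertFromDiffSpec material function.
# The material uses the Default Lit shading model, and its graph has no textures or other inputs.
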Material /Engine/EngineDebugMaterials/LevelColorationLitMaterial
graph {
 properties {pixelDepthOffsetMode=PDOM_Legacy, usedShadingModels=Default Lit}
 nodes [
  MaterialExpressionVectorParameter0 = VectorParameter() {parameterName=Color, defaultValue=<rgba 0.5 0.5 0.5 0>}
  MaterialFunctionCall(DiffuseColor=@MaterialExpressionVectorParameter0.RGB) {materialFunction=&ConvertFromDiffSpec}
 ]
}
