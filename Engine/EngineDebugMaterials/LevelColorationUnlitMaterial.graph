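# Engine debug material that, going by its name, serves as the unlit variant for level coloration views.
# It uses the Unlit shading model and is flagged for use with editor compositing.
# Its only node is a Color vector parameter, red by default, which material instances can override.
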Material /Engine/EngineDebugMaterials/LevelColorationUnlitMaterial
graph {
 properties {bUsedWithEditorCompositing=true, pixelDepthOffsetMode=PDOM_Legacy, shadingModel=MSM_Unlit, usedShadingModels=Unlit}
 nodes [MaterialExpressionVectorParameter0 = VectorParameter() {parameterName=Color, defaultValue=<rgba 1 0 0 0>}]
}
