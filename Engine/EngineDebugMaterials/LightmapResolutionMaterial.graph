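# LightmapResolutionMaterial is a Default Lit engine debug material that, by its name and parameter, visualizes lightmap resolution as a tiled texture.
# The LightMapResolution scalar multiplies the texture coordinates, setting the tiling of the DefaultDiffuse and DefaultNormal samples.
# Diffuse color and a scaled-down copy used as specular, labeled FresnelBaseReflectFraction, feed the ConvertFromDiffSpec material function.
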Material /Engine/EngineDebugMaterials/LightmapResolutionMaterial
graph {
 properties {pixelDepthOffsetMode=PDOM_Legacy, usedShadingModels=Default Lit}
 nodes [
  n1 = Multiply(A=TextureCoordinate(), B=ScalarParameter() {defaultValue=1, parameterName=LightMapResolution})
  n2 = TextureSample(UVs=@n1) {texture=/Engine/EngineMaterials/DefaultDiffuse}
  MaterialFunctionCall(DiffuseColor=@n2.RGB, SpecularColor=Multiply(A=@n2.RGB) {constA=1, constB=0.04, desc=FresnelBaseReflectFraction}) {materialFunction=&ConvertFromDiffSpec}
  TextureSample(UVs=@n1) {samplerType=SAMPLERTYPE_Normal, texture=/Engine/EngineMaterials/DefaultNormal}
 ]
}
