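# Engine debug material, Default Lit, that shows a colored grid laid out in lightmap UV space, by its name for lit level coloration.
# Lightmap UVs are scaled by half the red channel of the LightmapRes parameter on both axes, then sample DefaultWhiteGrid, which is tinted by the Color parameter.
# A feature level switch substitutes a constant for the lightmap UVs on ES2 and ES3_1 to avoid an ES2 error.
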Material /Engine/EngineDebugMaterials/MAT_LevelColorationLitLightmapUV
graph {
 properties {pixelDepthOffsetMode=PDOM_Legacy, usedShadingModels=Default Lit}
 nodes [
  Constant()
  Constant() {r=0.3}
  MaterialExpressionVectorParameter0 = VectorParameter() {parameterName=Color, defaultValue=<rgba 1 0 0 0>}
  n1 = VectorParameter() {parameterName=LightmapRes, defaultValue=<rgba 1024 512 0 1>}
  n2 = Multiply(A=AppendVector(A=@n1.R, B=@n1.R), B=Constant() {r=0.5})
  n3 = LightmapUVs()
  n4 = Constant()
  n5 = FeatureLevelSwitch(Default=@n3, ES2=@n4, ES3_1=@n4, SM4_REMOVED=@n3) {desc=Added to remove ES2 error}
  Multiply(
   A=@MaterialExpressionVectorParameter0.RGB
   B=TextureSample(UVs=Multiply(A=@n2, B=@n5)) {texture=/Engine/EngineMaterials/DefaultWhiteGrid}.RGB
  )
 ]
}
