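# Two-sided unlit debug material for geometry inspection that highlights back faces, green by default, against gray-shaded front faces.
# Each side fakes lighting in view space with an angle-and-tilt light, wrapped diffuse and screen-derivative normal cavity darkening.
# Front faces add Blinn-Phong specular and rim terms, while back-face rim and specular branches are left unconnected.
# TwoSidedSign blends the two, with eye-adaptation compensation only on the back-face color.
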Material /Engine/EngineDebugMaterials/M_GeoInsp_BackFace
graph {
 properties {
  bFullyRough=true
  bTangentSpaceNormal=false
  bUseEmissiveForDynamicAreaLighting=true
  bUsedWithEditorCompositing=true
  bUsedWithNanite=true
  shadingModel=MSM_Unlit
  twoSided=true
  usedShadingModels=Unlit
 }
 nodes [
  n1 = Normalize(VectorInput=Transform(None=PixelNormalWS()) {transformSourceType=TRANSFORMSOURCE_World, transformType=TRANSFORM_View})
  n2 = Multiply(A=ScalarParameter() {defaultValue=34.5, group=Backface, parameterName=Backface Light Angle Deg}, B=Constant() {r=0.0175})
  n3 = Normalize(VectorInput=MaterialFunctionCall(R=Sine(None=@n2), G=Cosine(None=@n2), B=ScalarParameter() {defaultValue=0.26, group=Backface, parameterName=Backface Light Tilt}) {materialFunction=&AppendMany}.RGB)
  n4 = DotProduct(A=@n1, B=@n3)
  n5 = ScalarParameter() {defaultValue=0.3, group=Backface, parameterName=Backface Wrap, sliderMax=0.6, sliderMin=0.2}
  n6 = Divide(A=Add(A=Saturate(None=@n4), B=@n5), B=Add(B=@n5) {constA=1})
  n7 = Multiply(
   A=Power(
    Base=Saturate(None=@n6)
    Exp=ScalarParameter() {defaultValue=1.2, group=Backface, parameterName=Backface Diffuse Exp, sliderMax=2, sliderMin=0.8}
   )
   B=VectorParameter() {parameterName=Backface Diffuse, defaultValue=<rgba 0 0.5417 0 1>}.RGB
  )
  n8 = DDX(Value=@n1)
  n9 = DotProduct(A=@n8, B=@n8)
  n10 = DDY(Value=@n1)
  n11 = DotProduct(A=@n10, B=@n10)
  n12 = Multiply(A=SquareRoot(None=Add(A=@n9, B=@n11)), B=ScalarParameter() {parameterName=Cavity Scale})
  n13 = Multiply(A=@n7, B=Saturate(None=OneMinus(None=@n12)))
  n14 = Normalize(VectorInput=Transform(None=CameraVectorWS()) {transformSourceType=TRANSFORMSOURCE_World, transformType=TRANSFORM_View})
  n15 = Saturate(None=DotProduct(A=@n1, B=@n14))
  n16 = Multiply(
   A=Power(Base=OneMinus(None=@n15), Exp=ScalarParameter() {defaultValue=1, group=Backface, parameterName=Backface Rim Power})
   B=ScalarParameter() {group=Backface, parameterName=Backface Rim Intensity}
  )
  Add(
   A=@n13
   B=MaterialFunctionCall(X=@n16, Y=@n16, Z=@n16) {materialFunction=&MakeFloat3}.Result
  ) {desc=No Rim}
  n17 = Normalize(VectorInput=Add(A=@n3, B=@n14))
  n18 = Power(Base=Saturate(None=DotProduct(A=@n1, B=@n17)), Exp=ScalarParameter() {defaultValue=9.5226, group=Backface, parameterName=Backface Gloss})
  n19 = ScalarParameter() {defaultValue=2.1147, group=Backface, parameterName=Backface Spec Intensity}
  Add(
   A=@n7
   B=Multiply(A=Multiply(A=@n18, B=@n19), B=VectorParameter() {group=Backface, parameterName=Backface Specular, defaultValue=<rgba 1 0.98 0.95 1>}.RGB)
  ) {desc=No Sepc highlight}
  n20 = Normalize(VectorInput=Transform(None=PixelNormalWS()) {transformSourceType=TRANSFORMSOURCE_World, transformType=TRANSFORM_View})
  n21 = Multiply(A=ScalarParameter() {defaultValue=5, parameterName=Light Angle Deg}, B=Constant() {r=0.0175})
  n22 = Normalize(VectorInput=MaterialFunctionCall(R=Sine(None=@n21), G=Cosine(None=@n21), B=ScalarParameter() {defaultValue=-0.355, parameterName=Light Tilt}) {materialFunction=&AppendMany}.RGB)
  n23 = Saturate(None=DotProduct(A=@n20, B=@n22))
  n24 = ScalarParameter() {defaultValue=0.2, parameterName=Wrap, sliderMax=0.6, sliderMin=0.2}
  n25 = Saturate(None=Divide(A=Add(A=@n23, B=@n24), B=Add(B=@n24) {constA=1}))
  n26 = ScalarParameter() {defaultValue=1.2, parameterName=Diffuse Exp, sliderMax=2, sliderMin=0.8}
  n27 = Normalize(VectorInput=Transform(None=CameraVectorWS()) {transformSourceType=TRANSFORMSOURCE_World, transformType=TRANSFORM_View})
  n28 = Normalize(VectorInput=Add(A=@n22, B=@n27))
  n29 = Power(Base=Saturate(None=DotProduct(A=@n20, B=@n28)), Exp=ScalarParameter() {defaultValue=40, parameterName=Gloss})
  n30 = ScalarParameter() {defaultValue=0.05, parameterName=Spec Intensity}
  n31 = Add(
   A=Multiply(A=Power(Base=@n25, Exp=@n26), B=VectorParameter() {parameterName=Diffuse, defaultValue=<rgba 0.5 0.5 0.5 1>}.RGB)
   B=Multiply(A=Multiply(A=@n29, B=@n30), B=VectorParameter() {parameterName=Specular, defaultValue=<rgba 1 0.98 0.95 1>}.RGB)
  )
  n32 = DDX(Value=@n20)
  n33 = DDY(Value=@n20)
  n34 = SquareRoot(None=Add(A=DotProduct(A=@n32, B=@n32), B=DotProduct(A=@n33, B=@n33)))
  n35 = ScalarParameter() {parameterName=Cavity Scale}
  n36 = Saturate(None=OneMinus(None=Multiply(A=@n34, B=@n35)))
  n37 = Saturate(None=DotProduct(A=@n20, B=@n27))
  n38 = Multiply(
   A=Power(Base=OneMinus(None=@n37), Exp=ScalarParameter() {defaultValue=1, parameterName=Rim Power})
   B=ScalarParameter() {parameterName=Rim Intensity}
  )
  LinearInterpolate(
   A=EyeAdaptationInverse(LightValueInput=@n13).EyeAdaptationInverse
   B=Add(
    A=Multiply(A=@n31, B=@n36)
    B=MaterialFunctionCall(X=@n38, Y=@n38, Z=@n38) {materialFunction=&MakeFloat3}.Result
   )
   Alpha=Saturate(None=TwoSidedSign())
  ) {constB=0}
 ]
}
