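# Unlit debug material that gives geometry a clay look, apparently for geometry inspection, with editor compositing support.
# It fakes a parameterized light in view space, so lighting follows the camera, combining wrapped diffuse with Blinn-Phong specular.
# Screen-space derivatives of the view-space normal darken high-curvature areas as a cavity term, then a Fresnel-style rim is added.
# Cavity and rim strengths default to zero, leaving both effects off.
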Material /Engine/EngineDebugMaterials/M_GeoInsp_Clay
graph {
 properties {bUsedWithEditorCompositing=true, shadingModel=MSM_Unlit, usedShadingModels=Unlit}
 nodes [
  n1 = Normalize(VectorInput=Transform(None=PixelNormalWS()) {transformSourceType=TRANSFORMSOURCE_World, transformType=TRANSFORM_View})
  n2 = Multiply(A=ScalarParameter() {defaultValue=5, parameterName=Light Angle Deg}, B=Constant() {r=0.0175})
  n3 = Normalize(VectorInput=MaterialFunctionCall(R=Sine(None=@n2), G=Cosine(None=@n2), B=ScalarParameter() {defaultValue=-0.355, parameterName=Light Tilt}) {materialFunction=&AppendMany}.RGB)
  n4 = DotProduct(A=@n1, B=@n3)
  n5 = ScalarParameter() {defaultValue=0.3, parameterName=Wrap, sliderMax=0.6, sliderMin=0.2}
  n6 = Divide(A=Add(A=Saturate(None=@n4), B=@n5), B=Add(B=@n5) {constA=1})
  n7 = Multiply(
   A=Power(Base=Saturate(None=@n6), Exp=ScalarParameter() {defaultValue=1.35, parameterName=Diffuse Exp, sliderMax=2, sliderMin=0.8})
   B=VectorParameter() {parameterName=Diffuse, defaultValue=<rgba 0.374 0.2556 0.1765 1>}.RGB
  )
  n8 = Normalize(VectorInput=Transform(None=CameraVectorWS()) {transformSourceType=TRANSFORMSOURCE_World, transformType=TRANSFORM_View})
  n9 = Add(A=@n3, B=@n8)
  n10 = Saturate(None=DotProduct(A=@n1, B=Normalize(VectorInput=@n9)))
  n11 = ScalarParameter() {defaultValue=40, parameterName=Gloss}
  n12 = Multiply(
   A=Multiply(A=Power(Base=@n10, Exp=@n11), B=ScalarParameter() {defaultValue=0.03, parameterName=Spec Intensity})
   B=VectorParameter() {parameterName=Specular, defaultValue=<rgba 1 1 1 0>}.RGB
  )
  n13 = DDX(Value=@n1)
  n14 = DDY(Value=@n1)
  n15 = Add(A=DotProduct(A=@n13, B=@n13), B=DotProduct(A=@n14, B=@n14))
  n16 = OneMinus(None=Multiply(A=SquareRoot(None=@n15), B=ScalarParameter() {parameterName=Cavity Scale}))
  n17 = Saturate(None=DotProduct(A=@n1, B=@n8))
  n18 = Multiply(
   A=Power(Base=OneMinus(None=@n17), Exp=ScalarParameter() {defaultValue=2, parameterName=Rim Power})
   B=ScalarParameter() {parameterName=Rim Intensity}
  )
  Add(
   A=Multiply(A=Add(A=@n7, B=@n12), B=Saturate(None=@n16))
   B=MaterialFunctionCall(X=@n18, Y=@n18, Z=@n18) {materialFunction=&MakeFloat3}.Result
  )
 ]
}
