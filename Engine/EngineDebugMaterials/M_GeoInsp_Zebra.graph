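# Unlit debug material that draws two-color zebra stripes from the reflection vector, by its name for inspecting geometry surfaces.
# The normalized reflection vector is dotted with a parameterized channel mask of itself, then scaled, offset and wrapped by Frac into repeating bands.
# A hard Step on those bands blends the two stripe colors.
# A static switch offers derivative-widened SmoothStep edges, but the color blend shown reads the hard Step directly.
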Material /Engine/EngineDebugMaterials/M_GeoInsp_Zebra
graph {
 properties {
  bUsedWithEditorCompositing=true
  bUsedWithInstancedStaticMeshes=true
  bUsedWithStaticLighting=true
  pixelDepthOffsetMode=PDOM_Legacy
  shadingModel=MSM_Unlit
  usedShadingModels=Unlit
 }
 nodes [
  n1 = Normalize(VectorInput=ReflectionVectorWS())
  n2 = DotProduct(A=@n1, B=ChannelMaskParameter(None=@n1) {parameterName=XYZ}.R)
  n3 = ScalarParameter() {defaultValue=10, parameterName=Stripe Counts}
  n4 = Frac(None=Add(A=Multiply(A=@n2, B=@n3), B=ScalarParameter() {parameterName=Strip Offset}))
  n5 = Step(Y=@n4, X=ScalarParameter() {defaultValue=0.5, parameterName=Strip Width Balance})
  LinearInterpolate(
   A=VectorParameter() {parameterName=Color_Zebra_01, defaultValue=<rgba 1 1 1 1>}.RGB
   B=VectorParameter() {parameterName=Color_Zebra_02, defaultValue=<rgba 0.025 0.025 0.025 1>}.RGB
   Alpha=@n5
  )
  n6 = ScalarParameter() {defaultValue=0.5, parameterName=Strip Width Balance}
  n7 = DDX(Value=@n4)
  n8 = DDY(Value=@n4)
  n9 = Add(A=Add(A=Abs(None=@n7), B=Abs(None=@n8)), B=ScalarParameter() {parameterName=Softness})
  StaticSwitchParameter(True=SmoothStep(Min=Subtract(A=@n6, B=@n9), Max=Add(A=@n6, B=@n9), Value=@n4), False=@n5) {parameterName="Bend Soft - True, Hard - False"}
 ]
}
