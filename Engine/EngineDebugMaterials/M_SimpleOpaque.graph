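# M_SimpleOpaque is a minimal opaque, Default Lit material in the engine's debug materials folder.
# Its graph holds a single vector parameter, Color, with a mid-grey default that material instances can override.
# The graph shows no other expressions and no link from Color to a material output.
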Material /Engine/EngineDebugMaterials/M_SimpleOpaque
graph {
 properties {pixelDepthOffsetMode=PDOM_Legacy, usedShadingModels=Default Lit}
 nodes [VectorParameter() {parameterName=Color, defaultValue=<rgba 0.4 0.4 0.4 1>}]
}
