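# M_SimpleTranslucent is an engine debug material with translucent blending and Default Lit shading.
# Its only listed node is a Color vector parameter, defaulting to half-alpha red, which material instances can override.
# The graph shows no links from that parameter to the material outputs.
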Material /Engine/EngineDebugMaterials/M_SimpleTranslucent
graph {
 properties {blendMode=BLEND_Translucent, pixelDepthOffsetMode=PDOM_Legacy, usedShadingModels=Default Lit}
 nodes [VectorParameter() {parameterName=Color, defaultValue=<rgba 1 0 0 0.5>}]
}
