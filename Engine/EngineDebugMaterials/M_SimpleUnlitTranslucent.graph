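# Engine debug material that provides a simple unlit, translucent surface.
# Its graph holds a single vector parameter, Color, defaulting to half-alpha red, which material instances can override.
# The file lists no links from that parameter to the material outputs.
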Material /Engine/EngineDebugMaterials/M_SimpleUnlitTranslucent
graph {
 properties {blendMode=BLEND_Translucent, pixelDepthOffsetMode=PDOM_Legacy, shadingModel=MSM_Unlit, usedShadingModels=Unlit}
 nodes [VectorParameter() {parameterName=Color, defaultValue=<rgba 1 0 0 0.5>}]
}
