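# Post-process debug material that renders a volume texture as an isosurface over the scene.
# A Custom HLSL node transforms camera rays into volume-local space, intersects the bounds box, then sphere-traces using texture alpha as distance, rejecting hits behind scene depth.
# Hit RGB replaces the scene color through an alpha mask.
# A debug parameter fills the bounds white instead, and a CustomDepth lookup sits unconnected.
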Material /Engine/EngineDebugMaterials/M_VolumeRenderSphereTracePP
graph {
 properties {materialDomain=MD_PostProcess, pixelDepthOffsetMode=PDOM_Legacy, usedShadingModels=Default Lit}
 nodes [
  n1 = SceneTexture() {sceneTextureId=PPI_PostProcessInput0}
  MaterialExpressionCameraPositionWS_0 = CameraPositionWS()
  MaterialExpressionNormalize_0 = Normalize(VectorInput=Subtract(A=WorldPosition().XYZ, B=@MaterialExpressionCameraPositionWS_0))
  MaterialExpressionTextureObjectParameter_0 = TextureObjectParameter() {parameterName=Volume To Render, samplerType=SAMPLERTYPE_LinearColor, texture=&/VolumeToRender}
  MaterialExpressionVectorParameter_0 = VectorParameter() {parameterName=Min Bounds}
  MaterialExpressionVectorParameter_1 = VectorParameter() {parameterName=Max Bounds}
  MaterialExpressionScalarParameter_2 = ScalarParameter() {defaultValue=0.0001, parameterName=Surface Tolerance}
  MaterialExpressionScalarParameter_3 = ScalarParameter() {parameterName=Isovalue}
  MaterialExpressionConstant_0 = Constant() {desc=Number of ray march steps, r=30}
  MaterialExpressionScalarParameter_0 = ScalarParameter() {defaultValue=0.25, parameterName=Voxel Size}
  MaterialExpressionScalarParameter_1 = ScalarParameter() {defaultValue=1, parameterName=Step Size Mult}
  MaterialExpressionVectorParameter_5 = VectorParameter() {parameterName=WorldToLocalTranslation}
  MaterialExpressionVectorParameter_2 = VectorParameter() {parameterName=WorldToLocalc0}
  MaterialExpressionVectorParameter_3 = VectorParameter() {parameterName=WorldToLocalc1}
  MaterialExpressionVectorParameter_4 = VectorParameter() {parameterName=WorldToLocalc2}
  MaterialExpressionMaterialFunctionCall_1 = MaterialFunctionCall() {materialFunction=&CameraDirectionVector}
  MaterialExpressionSceneDepth_0 = SceneDepth()
  MaterialExpressionScalarParameter_4 = ScalarParameter() {parameterName=Debug BBox}
  n2 = Custom!(
   RayStart=@MaterialExpressionCameraPositionWS_0
   RayDir=@MaterialExpressionNormalize_0
   VolumeTexture=@MaterialExpressionTextureObjectParameter_0
   MinBounds=@MaterialExpressionVectorParameter_0.RGB
   MaxBounds=@MaterialExpressionVectorParameter_1.RGB
   SurfaceDistanceTolerance=@MaterialExpressionScalarParameter_2
   Isovalue=@MaterialExpressionScalarParameter_3
   NumSteps=@MaterialExpressionConstant_0
   VoxelSize=@MaterialExpressionScalarParameter_0
   StepSizeMult=@MaterialExpressionScalarParameter_1
   WorldToLocalTranslation=@MaterialExpressionVectorParameter_5.RGB
   WorldToLocalc0=@MaterialExpressionVectorParameter_2.RGB
   WorldToLocalc1=@MaterialExpressionVectorParameter_3.RGB
   WorldToLocalc2=@MaterialExpressionVectorParameter_4.RGB
   CameraDirectionVector=@MaterialExpressionMaterialFunctionCall_1.Result
   SceneDepth=@MaterialExpressionSceneDepth_0.R
   DebugBBox=@MaterialExpressionScalarParameter_4
  ) {
   code="float4 ret = {0,0,0, 0};\r\n\r\nfloat3 hit = {1,1,1};\r\n\r\nfloat3x3 WorldToLocalRotation = \r\n{\r\nWorldToLocalc0.x, WorldToLocalc0.y, WorldToLocalc0.z,\r\nWorldToLocalc1.x, WorldToLocalc1.y, WorldToLocalc1.z,\r\nWorldToLocalc2.x, WorldToLocalc2.y, WorldToLocalc2.z\r\n};\r\n\r\nfloat WorldDepth = SceneDepth / dot(RayDir, CameraDirectionVector);\r\n\r\n\r\nRayStart = mul(WorldToLocalRotation, RayStart);\r\nRayStart += WorldToLocalTranslation;\r\n\r\nRayDir = mul(WorldToLocalRotation, RayDir);\r\n\r\n//\r\n//  Ray aabbox intersection\r\n//\r\n\r\nfloat3 dirfrac = 1.0f / RayDir;\r\n//dirfrac.x = 1.0f / RayDir.x;\r\n//dirfrac.y = 1.0f / RayDir.y;\r\n//dirfrac.z = 1.0f / RayDir.z;\r\n\r\nfloat t = 0;\r\n\r\nfloat t1 = (MinBounds.x - RayStart.x)*dirfrac.x;\r\nfloat t2 = (MaxBounds.x - RayStart.x)*dirfrac.x;\r\nfloat t3 = (MinBounds.y - RayStart.y)*dirfrac.y;\r\nfloat t4 = (MaxBounds.y - RayStart.y)*dirfrac.y;\r\nfloat t5 = (MinBounds.z - RayStart.z)*dirfrac.z;\r\nfloat t6 = (MaxBounds.z - RayStart.z)*dirfrac.z;\r\n\r\nfloat tmin = max(max(min(t1, t2), min(t3, t4)), min(t5, t6));\r\nfloat tmax = min(min(max(t1, t2), max(t3, t4)), max(t5, t6));\r\n\r\n// if tmax < 0, ray (line) is intersecting AABB, but the whole AABB is behind us\r\nif (tmax < 0)\r\n{\r\n    t = tmax;\r\n    return ret;\r\n}\r\n\r\n// if tmin > tmax, ray doesn't intersect AABB\r\nif (tmin > tmax)\r\n{\r\n    t = tmax;\r\n    return ret;\r\n}\r\n\r\nt = tmin;\r\n\r\nif (DebugBBox > 1e-3) {\r\n  float4 rettmp = {1,1,1,1};\r\n  return rettmp;\r\n}\r\n\r\n// expose step size\r\n// do sphere tracing\r\n// other rendering options (slices, etc)\r\n// read in transform info from pieces\r\n\r\nfloat3 currPos = RayStart + RayDir * t;\r\n\r\n// @todo: optimization to factor this out\r\nfloat3 boxSize = MaxBounds - MinBounds;\r\n\r\nfloat prevDistToSurface = 1000000;\r\n\r\n// our surface tolerance is the voxel size times our multiplier\r\nfloat tol = VoxelSize*SurfaceDistanceTolerance;\r\n\r\n      \r\nfor (int i = 0; i < NumSteps; ++i) {\r\n\r\n  // position in local space \r\n  float3 localPos = currPos - MinBounds;\r\n  localPos /= boxSize;\r\n  \r\n  float dist = VoxelSize;\r\n\r\n  // @todo: since we are intersecting the bbox, do we need to check this?\r\n  if (\r\n    localPos[0] > 1e-8 &&&& localPos[0] < 1 &&&&\r\n    localPos[1] > 1e-8 &&&& localPos[1] < 1 &&&&\r\n    localPos[2] > 1e-8 &&&& localPos[2] < 1\r\n  ) {\r\n\r\n    float4 val = Texture3DSampleLevel(VolumeTexture, VolumeTextureSampler, localPos, 0);\r\n    float currDistToSurface = val[3]-Isovalue;\r\n    // terminate if sufficiently close\r\n    if (abs(currDistToSurface) < tol &&&& t < WorldDepth) {\r\n      ret = val;\r\n      ret[3] = 1;\r\n      break;\r\n    }\r\n\r\n    // sphere trace\r\n    dist = currDistToSurface;\r\n    prevDistToSurface = currDistToSurface;\r\n  }\r\n\r\n  t += dist * StepSizeMult;\r\n  currPos = RayStart + RayDir * t;\r\n}\r\n\r\nreturn ret;"
   outputType=CMOT_Float4
   inputs [
    {inputName=RayStart, input {expression=&self:MaterialExpressionCameraPositionWS_0}}
    {inputName=RayDir, input {expression=&self:MaterialExpressionNormalize_0}}
    {
     inputName=VolumeTexture
     input {expression=&self:MaterialExpressionTextureObjectParameter_0}
    }
    {
     inputName=MinBounds
     input {expression=&self:MaterialExpressionVectorParameter_0, mask=1, maskB=1, maskG=1, maskR=1}
    }
    {
     inputName=MaxBounds
     input {expression=&self:MaterialExpressionVectorParameter_1, mask=1, maskB=1, maskG=1, maskR=1}
    }
    {
     inputName=SurfaceDistanceTolerance
     input {expression=&self:MaterialExpressionScalarParameter_2}
    }
    {inputName=Isovalue, input {expression=&self:MaterialExpressionScalarParameter_3}}
    {inputName=NumSteps, input {expression=&self:MaterialExpressionConstant_0}}
    {inputName=VoxelSize, input {expression=&self:MaterialExpressionScalarParameter_0}}
    {inputName=StepSizeMult, input {expression=&self:MaterialExpressionScalarParameter_1}}
    {
     inputName=WorldToLocalTranslation
     input {expression=&self:MaterialExpressionVectorParameter_5, mask=1, maskB=1, maskG=1, maskR=1}
    }
    {
     inputName=WorldToLocalc0
     input {expression=&self:MaterialExpressionVectorParameter_2, mask=1, maskB=1, maskG=1, maskR=1}
    }
    {
     inputName=WorldToLocalc1
     input {expression=&self:MaterialExpressionVectorParameter_3, mask=1, maskB=1, maskG=1, maskR=1}
    }
    {
     inputName=WorldToLocalc2
     input {expression=&self:MaterialExpressionVectorParameter_4, mask=1, maskB=1, maskG=1, maskR=1}
    }
    {
     inputName=CameraDirectionVector
     input {expression=&self:MaterialExpressionMaterialFunctionCall_1}
    }
    {
     inputName=SceneDepth
     input {expression=&self:MaterialExpressionSceneDepth_0, mask=1, maskR=1}
    }
    {inputName=DebugBBox, input {expression=&self:MaterialExpressionScalarParameter_4}}
   ]
  }
  n3 = ComponentMask(None=@n2) {a=true}
  Add(
   A=Multiply(A=ComponentMask(None=@n1.Color) {b=true, g=true, r=true}, B=OneMinus(None=@n3))
   B=ComponentMask(None=@n2) {b=true, g=true, r=true}
  )
  SceneTexture() {sceneTextureId=PPI_CustomDepth}
 ]
}
