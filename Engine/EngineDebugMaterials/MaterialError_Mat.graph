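# MaterialError_Mat is a Default Lit engine debug material built on the MaterialError texture, apparently serving as a visual marker for material errors.
# The texture is sampled with UVs tiled twenty times on both axes, producing a dense repeating pattern.
# Its RGB feeds the diffuse input of the ConvertFromDiffSpec material function and, separately, a multiply that halves its brightness.
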Material /Engine/EngineDebugMaterials/MaterialError_Mat
graph {
 properties {pixelDepthOffsetMode=PDOM_Legacy, usedShadingModels=Default Lit}
 nodes [
  n1 = TextureSample(UVs=TextureCoordinate() {uTiling=20, vTiling=20}) {texture=&/MaterialError}
  MaterialFunctionCall(DiffuseColor=@n1.RGB) {materialFunction=&ConvertFromDiffSpec}
  Multiply(A=@n1.RGB, B=Constant() {r=0.5})
 ]
}
