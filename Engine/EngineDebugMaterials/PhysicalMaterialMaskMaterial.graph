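# PhysicalMaterialMaskMaterial is an engine debug material, by its name for physical material masks, using two-sided Default Lit shading.
# Its graph multiplies a white-default Color vector parameter by the LinearColor texture parameter, which defaults to a white grid texture.
# The file shows no connection from that product to a material output.
# Usage flags cover skeletal and instanced static meshes, mesh deformers, geometry caches and geometry collections.
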Material /Engine/EngineDebugMaterials/PhysicalMaterialMaskMaterial
graph {
 properties {
  bUsedWithGeometryCache=true
  bUsedWithGeometryCollections=true
  bUsedWithInstancedStaticMeshes=true
  bUsedWithMeshDeformer=true
  bUsedWithSkeletalMesh=true
  pixelDepthOffsetMode=PDOM_Legacy
  twoSided=true
  usedShadingModels=Default Lit
 }
 nodes [Multiply(
   A=VectorParameter() {parameterName=Color, defaultValue=<rgba 1 1 1 1>}.RGB
   B=TextureSampleParameter2D() {parameterName=LinearColor, texture=/Engine/EngineMaterials/DefaultWhiteGrid_Low}.RGB
  )]
}
