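# Engine debug material for the shaded level coloration lit view, tinting surfaces with a Color vector parameter.
# It transforms the tangent-space Z vector, takes absolute dot products against X, Y and Z axes, and appends them into a vector.
# Desaturating that vector gives orientation-based shading, which multiplies Color and feeds the DiffuseColor input of ConvertFromDiffSpec under Default Lit shading.
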Material /Engine/EngineDebugMaterials/ShadedLevelColorationLitMaterial
graph {
 properties {pixelDepthOffsetMode=PDOM_Legacy, usedShadingModels=Default Lit}
 nodes [
  n1 = Transform(None=Constant3Vector() {constant=<rgba 0 0 1 0>})
  n2 = DotProduct(A=@n1, B=Constant3Vector() {desc=X, constant=<rgba 1 0 0 0>})
  n3 = DotProduct(A=@n1, B=Constant3Vector() {desc=Y, constant=<rgba 0 1 0 0>})
  n4 = Constant3Vector() {desc=Z, constant=<rgba 0 0 1 0>}
  n5 = AppendVector(A=AppendVector(A=Abs(None=@n2), B=Abs(None=@n3)), B=Abs(None=DotProduct(A=@n1, B=@n4)))
  MaterialExpressionVectorParameter0 = VectorParameter() {parameterName=Color, defaultValue=<rgba 1 0 0 0>}
  MaterialFunctionCall(DiffuseColor=Multiply(A=Desaturation(None=@n5), B=@MaterialExpressionVectorParameter0.RGB)) {materialFunction=&ConvertFromDiffSpec}
 ]
}
