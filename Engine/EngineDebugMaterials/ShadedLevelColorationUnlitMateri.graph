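# Unlit engine debug material that draws level coloration as a flat color with fake directional shading.
# It transforms a constant Z vector, by default from tangent to world space to give the surface normal, and takes absolute dot products with the X, Y and Z axes.
# Desaturating that vector weights each axis differently, producing an orientation-based brightness without lights.
# The brightness multiplies the Color parameter, which defaults to red.
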Material /Engine/EngineDebugMaterials/ShadedLevelColorationUnlitMateri
graph {
 properties {pixelDepthOffsetMode=PDOM_Legacy, shadingModel=MSM_Unlit, usedShadingModels=Unlit}
 nodes [
  n1 = Transform(None=Constant3Vector() {constant=<rgba 0 0 1 0>})
  n2 = Constant3Vector() {desc=X, constant=<rgba 1 0 0 0>}
  n3 = Constant3Vector() {desc=Y, constant=<rgba 0 1 0 0>}
  n4 = AppendVector(A=Abs(None=DotProduct(A=@n1, B=@n2)), B=Abs(None=DotProduct(A=@n1, B=@n3)))
  n5 = Abs(None=DotProduct(A=@n1, B=Constant3Vector() {desc=Z, constant=<rgba 0 0 1 0>}))
  Multiply(A=Desaturation(None=AppendVector(A=@n4, B=@n5)), B=VectorParameter() {parameterName=Color, defaultValue=<rgba 1 0 0 0>}.RGB)
 ]
}
