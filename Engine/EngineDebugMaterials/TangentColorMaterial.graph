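# TangentColorMaterial is an engine debug material flagged for use with skeletal meshes and mesh deformers.
# Its graph samples the TangentColorMap texture using texture coordinates passed through a red-green component mask.
# It uses the Default Lit shading model, and the graph shows no connection from the sample to a material output.
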Material /Engine/EngineDebugMaterials/TangentColorMaterial
graph {
 properties {bUsedWithMeshDeformer=true, bUsedWithSkeletalMesh=true, pixelDepthOffsetMode=PDOM_Legacy, usedShadingModels=Default Lit}
 nodes [TextureSample(UVs=ComponentMask(None=TextureCoordinate()) {g=true, r=true}) {texture=&/TangentColorMap}]
}
