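# Engine debug material for drawing terrain collision, going by its name and folder.
# It is an unlit, translucent material whose only node is a CollisionColor vector parameter, defaulting to semi-transparent red.
# The graph shows no connection from that parameter to a material output.
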Material /Engine/EngineDebugMaterials/TerrainCollisionMaterial
graph {
 properties {blendMode=BLEND_Translucent, pixelDepthOffsetMode=PDOM_Legacy, shadingModel=MSM_Unlit, usedShadingModels=Unlit}
 nodes [VectorParameter() {parameterName=CollisionColor, defaultValue=<rgba 1 0 0 0.75>}]
}
