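# Unlit engine debug material for viewing mesh paint texture colors on static, skeletal, spline and geometry collection meshes.
# A Switch on the mesh paint coordinate index picks among UV0–UV3 to sample the mesh paint texture.
# An If node compares the Color parameter's max(R, B) to its alpha, choosing Color-masked RGB or, when lower, grayscale texture alpha.
# Green is absent from that max, as R feeds it twice.
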Material /Engine/EngineDebugMaterials/TextureColorViewMode
graph {
 expressions {
  MaterialExpressionConstant_0 {class=MaterialExpressionConstant, properties {r=1}}
  MaterialExpressionIf_4 {
   class=MaterialExpressionIf
   inputs [
    A {from=MaterialExpressionMax_1, output=""}
    B {from=MaterialExpressionVectorParameter_0, output=A}
    "A > B" {from=MaterialExpressionMaterialFunctionCall_0, output=Result}
    "A == B" {from=MaterialExpressionMaterialFunctionCall_0, output=Result}
    "A < B" {from=MaterialExpressionMaterialFunctionCall_1, output=Result}
   ]
   properties {}
  }
  MaterialExpressionMaterialFunctionCall_0 {
   class=MaterialExpressionMaterialFunctionCall
   inputs [
    X {from=MaterialExpressionMultiply_0}
    Y {from=MaterialExpressionMultiply_1}
    Z {from=MaterialExpressionMultiply_2}
    A {from=MaterialExpressionConstant_0}
   ]
   properties {materialFunction=&MakeFloat4}
  }
  MaterialExpressionMaterialFunctionCall_1 {
   class=MaterialExpressionMaterialFunctionCall
   inputs [
    X {from=MaterialExpressionTextureSample_0, output=A}
    Y {from=MaterialExpressionTextureSample_0, output=A}
    Z {from=MaterialExpressionTextureSample_0, output=A}
    A {from=MaterialExpressionConstant_0}
   ]
   properties {materialFunction=&MakeFloat4}
  }
  MaterialExpressionMax_0 {
   class=MaterialExpressionMax
   inputs [
    A {from=MaterialExpressionVectorParameter_0, output=R}
    B {from=MaterialExpressionVectorParameter_0, output=R}
   ]
  }
  MaterialExpressionMax_1 {
   class=MaterialExpressionMax
   inputs [
    A {from=MaterialExpressionMax_0}
    B {from=MaterialExpressionVectorParameter_0, output=B}
   ]
  }
  MaterialExpressionMeshPaintTextureCoordinateIndex_0 {class=MaterialExpressionMeshPaintTextureCoordinateIndex, inputs [], properties {}}
  MaterialExpressionMeshPaintTextureObject_0 {class=MaterialExpressionMeshPaintTextureObject, inputs [], properties {}}
  MaterialExpressionMultiply_0 {
   class=MaterialExpressionMultiply
   inputs [
    A {from=MaterialExpressionVectorParameter_0, output=R}
    B {from=MaterialExpressionTextureSample_0, output=R}
   ]
  }
  MaterialExpressionMultiply_1 {
   class=MaterialExpressionMultiply
   inputs [
    A {from=MaterialExpressionVectorParameter_0, output=G}
    B {from=MaterialExpressionTextureSample_0, output=G}
   ]
  }
  MaterialExpressionMultiply_2 {
   class=MaterialExpressionMultiply
   inputs [
    A {from=MaterialExpressionVectorParameter_0, output=B}
    B {from=MaterialExpressionTextureSample_0, output=B}
   ]
  }
  MaterialExpressionSwitch_0 {
   class=MaterialExpressionSwitch
   inputs [
    SwitchValue {from=MaterialExpressionMeshPaintTextureCoordinateIndex_0, output=""}
    Default {from=MaterialExpressionTextureCoordinate_0, output=""}
    None {from=MaterialExpressionTextureCoordinate_0, output=""}
    None {from=MaterialExpressionTextureCoordinate_2, output=""}
    None {from=MaterialExpressionTextureCoordinate_1, output=""}
    None {from=MaterialExpressionTextureCoordinate_4, output=""}
   ]
   properties {inputs [
     {
      inputName=None
      input {
       expression=&self:MaterialExpressionTextureCoordinate_0
       inputName=None
       mask=0
       maskA=0
       maskB=0
       maskG=0
       maskR=0
       outputIndex=0
      }
     }
     {
      inputName=None
      input {
       expression=&self:MaterialExpressionTextureCoordinate_2
       inputName=None
       mask=0
       maskA=0
       maskB=0
       maskG=0
       maskR=0
       outputIndex=0
      }
     }
     {
      inputName=None
      input {
       expression=&self:MaterialExpressionTextureCoordinate_1
       inputName=None
       mask=0
       maskA=0
       maskB=0
       maskG=0
       maskR=0
       outputIndex=0
      }
     }
     {
      inputName=None
      input {
       expression=&self:MaterialExpressionTextureCoordinate_4
       inputName=None
       mask=0
       maskA=0
       maskB=0
       maskG=0
       maskR=0
       outputIndex=0
      }
     }
    ]}
  }
  MaterialExpressionTextureCoordinate_0 {class=MaterialExpressionTextureCoordinate}
  MaterialExpressionTextureCoordinate_1 {class=MaterialExpressionTextureCoordinate, properties {coordinateIndex=2}}
  MaterialExpressionTextureCoordinate_2 {class=MaterialExpressionTextureCoordinate, properties {coordinateIndex=1}}
  MaterialExpressionTextureCoordinate_4 {class=MaterialExpressionTextureCoordinate, properties {coordinateIndex=3}}
  MaterialExpressionTextureSample_0 {
   class=MaterialExpressionTextureSample
   inputs [
    UVs {from=MaterialExpressionSwitch_0}
    Tex {from=MaterialExpressionMeshPaintTextureObject_0}
    Apply View MipBias {from=""}
   ]
   properties {samplerSource=SSM_Clamp_WorldGroupSettings}
  }
  MaterialExpressionVectorParameter_0 {class=MaterialExpressionVectorParameter, properties {parameterName=Color, defaultValue=<rgba 1 1 1 1>}}
 }
 properties {
  bUsedWithGeometryCollections=true
  bUsedWithMeshDeformer=true
  bUsedWithMorphTargets=true
  bUsedWithSkeletalMesh=true
  bUsedWithSplineMeshes=true
  pixelDepthOffsetMode=PDOM_Legacy
  shadingModel=MSM_Unlit
  usedShadingModels=Unlit
 }
}
