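# Engine debug material for two-texture paint blending with color, a lit surface usable on skeletal meshes and mesh deformers.
# A PaintTexture parameter on UV channel 1 blends Diffuse2 to Diffuse per channel and also tints the result.
# The same mask blends two normal maps, scaled by a vector parameter.
# Legacy diffuse/specular inputs go through ConvertFromDiffSpec, specular power through PowerToRoughness, and emissive has a multiplier.
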Material /Engine/EngineDebugMaterials/TexturePaint_2Tex_Color
graph {
 properties {
  bUsedWithMeshDeformer=true
  bUsedWithSkeletalMesh=true
  bUsedWithStaticLighting=true
  pixelDepthOffsetMode=PDOM_Legacy
  usedShadingModels=Default Lit
 }
 nodes [
  MaterialFunctionCall!(Power=ScalarParameter() {defaultValue=15, parameterName=SpecPower}) {materialFunction=/Engine/Functions/Engine_MaterialFunctions01/Shading/PowerToRoughness}
  n1 = TextureSampleParameter2D() {parameterName=Diffuse2, texture=/Engine/Engine_MI_Shaders/T_Reflection}
  n2 = TextureSampleParameter2D() {parameterName=Diffuse, texture=&T_Reflection}
  n3 = TextureSampleParameter2D(UVs=TextureCoordinate() {coordinateIndex=1}) {parameterName=PaintTexture, texture=/Engine/EngineResources/WhiteSquareTexture}
  n4 = TextureSampleParameter2D() {parameterName=Spec, texture=&T_Reflection}
  n5 = ScalarParameter() {defaultValue=1, parameterName=SpecularMultValue}
  MaterialFunctionCall(
   DiffuseColor=Multiply(A=LinearInterpolate(A=@n1.RGB, B=@n2.RGB, Alpha=@n3.RGB), B=@n3.RGB)
   SpecularColor=Multiply(A=Multiply(A=@n4.RGB, B=@n5)) {constA=1, constB=0.04, desc=FresnelBaseReflectFraction}
  ) {materialFunction=&ConvertFromDiffSpec}
  Multiply(
   A=LinearInterpolate(
    A=TextureSampleParameter2D() {parameterName=Normal2, samplerType=SAMPLERTYPE_Normal, texture=/Engine/Engine_MI_Shaders/T_Normal}.RGB
    B=TextureSampleParameter2D() {parameterName=Normal, samplerType=SAMPLERTYPE_Normal, texture=&T_Normal}.RGB
    Alpha=@n3.RGB
   )
   B=VectorParameter() {parameterName=NormalMultiply, defaultValue=<rgba 1 1 1 1>}.RGB
  )
  Multiply(
   A=TextureSampleParameter2D() {parameterName=Emissive, texture=/Engine/Engine_MI_Shaders/T_Emissive}.RGB
   B=ScalarParameter() {defaultValue=10, parameterName=EmissiveMultiply}
  )
 ]
}
