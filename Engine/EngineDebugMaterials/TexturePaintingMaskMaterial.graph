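# Unlit debug material, named for texture painting masks, that shows either the tinted color channels or the alpha channel of a texture parameter.
# A UVChannel parameter drives a Switch over four texture coordinate sets for sampling LinearColor.
# An If compares the larger of Color's red and blue (green is omitted) with its alpha.
# Greater or equal yields texture RGB multiplied by Color, while less yields texture alpha as grayscale.
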Material /Engine/EngineDebugMaterials/TexturePaintingMaskMaterial
graph {
 expressions {
  MaterialExpressionConstant_1 {class=MaterialExpressionConstant, properties {r=1}}
  MaterialExpressionIf_0 {
   class=MaterialExpressionIf
   inputs [
    A {from=MaterialExpressionMax_3, output=""}
    B {from=MaterialExpressionVectorParameter_2, output=A}
    "A > B" {from=MaterialExpressionMaterialFunctionCall_1, output=Result}
    "A == B" {from=MaterialExpressionMaterialFunctionCall_1, output=Result}
    "A < B" {from=MaterialExpressionMaterialFunctionCall_3, output=Result}
   ]
   properties {}
  }
  MaterialExpressionMaterialFunctionCall_1 {
   class=MaterialExpressionMaterialFunctionCall
   inputs [
    X {from=MaterialExpressionMultiply_4}
    Y {from=MaterialExpressionMultiply_5}
    Z {from=MaterialExpressionMultiply_7}
    A {from=MaterialExpressionConstant_1}
   ]
   properties {materialFunction=&MakeFloat4}
  }
  MaterialExpressionMaterialFunctionCall_3 {
   class=MaterialExpressionMaterialFunctionCall
   inputs [
    X {from=MaterialExpressionTextureSampleParameter2D_1, output=A}
    Y {from=MaterialExpressionTextureSampleParameter2D_1, output=A}
    Z {from=MaterialExpressionTextureSampleParameter2D_1, output=A}
    A {from=MaterialExpressionConstant_1}
   ]
   properties {materialFunction=&MakeFloat4}
  }
  MaterialExpressionMax_2 {
   class=MaterialExpressionMax
   inputs [
    A {from=MaterialExpressionVectorParameter_2, output=R}
    B {from=MaterialExpressionVectorParameter_2, output=R}
   ]
  }
  MaterialExpressionMax_3 {
   class=MaterialExpressionMax
   inputs [
    A {from=MaterialExpressionMax_2}
    B {from=MaterialExpressionVectorParameter_2, output=B}
   ]
  }
  MaterialExpressionMultiply_4 {
   class=MaterialExpressionMultiply
   inputs [
    A {from=MaterialExpressionVectorParameter_2, output=R}
    B {from=MaterialExpressionTextureSampleParameter2D_1, output=R}
   ]
  }
  MaterialExpressionMultiply_5 {
   class=MaterialExpressionMultiply
   inputs [
    A {from=MaterialExpressionVectorParameter_2, output=G}
    B {from=MaterialExpressionTextureSampleParameter2D_1, output=G}
   ]
  }
  MaterialExpressionMultiply_7 {
   class=MaterialExpressionMultiply
   inputs [
    A {from=MaterialExpressionVectorParameter_2, output=B}
    B {from=MaterialExpressionTextureSampleParameter2D_1, output=B}
   ]
  }
  MaterialExpressionScalarParameter_1 {class=MaterialExpressionScalarParameter, properties {parameterName=UVChannel}}
  MaterialExpressionSwitch_0 {
   class=MaterialExpressionSwitch
   inputs [
    SwitchValue {from=MaterialExpressionScalarParameter_1, output=""}
    Default {from=MaterialExpressionTextureCoordinate_0, output=""}
    None {from=MaterialExpressionTextureCoordinate_0, output=""}
    None {from=MaterialExpressionTextureCoordinate_2, output=""}
    None {from=MaterialExpressionTextureCoordinate_1, output=""}
    None {from=MaterialExpressionTextureCoordinate_4, output=""}
   ]
   properties {inputs [
     {
      inputName=None
      input {
       expression=&self:MaterialExpressionTextureCoordinate_0
       inputName=None
       mask=0
       maskA=0
       maskB=0
       maskG=0
       maskR=0
       outputIndex=0
      }
     }
     {
      inputName=None
      input {
       expression=&self:MaterialExpressionTextureCoordinate_2
       inputName=None
       mask=0
       maskA=0
       maskB=0
       maskG=0
       maskR=0
       outputIndex=0
      }
     }
     {
      inputName=None
      input {
       expression=&self:MaterialExpressionTextureCoordinate_1
       inputName=None
       mask=0
       maskA=0
       maskB=0
       maskG=0
       maskR=0
       outputIndex=0
      }
     }
     {
      inputName=None
      input {
       expression=&self:MaterialExpressionTextureCoordinate_4
       inputName=None
       mask=0
       maskA=0
       maskB=0
       maskG=0
       maskR=0
       outputIndex=0
      }
     }
    ]}
  }
  MaterialExpressionTextureCoordinate_0 {class=MaterialExpressionTextureCoordinate}
  MaterialExpressionTextureCoordinate_1 {class=MaterialExpressionTextureCoordinate, properties {coordinateIndex=2}}
  MaterialExpressionTextureCoordinate_2 {class=MaterialExpressionTextureCoordinate, properties {coordinateIndex=1}}
  MaterialExpressionTextureCoordinate_4 {class=MaterialExpressionTextureCoordinate, properties {coordinateIndex=3}}
  MaterialExpressionTextureSampleParameter2D_1 {
   class=MaterialExpressionTextureSampleParameter2D
   inputs [
    UVs {from=MaterialExpressionSwitch_0, output=""}
    Apply View MipBias {from="", output=""}
   ]
   properties {constCoordinate=1, parameterName=LinearColor, texture=/Engine/EngineMaterials/DefaultWhiteGrid_Low}
  }
  MaterialExpressionVectorParameter_2 {class=MaterialExpressionVectorParameter, properties {parameterName=Color, defaultValue=<rgba 1 1 1 1>}}
 }
 properties {
  bUsedWithGeometryCollections=true
  bUsedWithMeshDeformer=true
  bUsedWithMorphTargets=true
  bUsedWithSkeletalMesh=true
  bUsedWithSplineMeshes=true
  pixelDepthOffsetMode=PDOM_Legacy
  shadingModel=MSM_Unlit
  usedShadingModels=Unlit
 }
}
