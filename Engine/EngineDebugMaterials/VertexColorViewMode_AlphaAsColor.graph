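# Engine debug material for the vertex color view mode that shows vertex alpha as color.
# A Custom node takes the vertex color alpha and replicates it into a float3, producing a grayscale value.
# The material is unlit and enabled for skeletal, spline and instanced static meshes, morph targets, mesh deformers and geometry collections.
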Material /Engine/EngineDebugMaterials/VertexColorViewMode_AlphaAsColor
graph {
 properties {
  bUsedWithGeometryCollections=true
  bUsedWithInstancedStaticMeshes=true
  bUsedWithMeshDeformer=true
  bUsedWithMorphTargets=true
  bUsedWithSkeletalMesh=true
  bUsedWithSplineMeshes=true
  pixelDepthOffsetMode=PDOM_Legacy
  shadingModel=MSM_Unlit
  usedShadingModels=Unlit
 }
 nodes [
  MaterialExpressionVertexColor_0 = VertexColor()
  Custom!(VertexOpacity=@MaterialExpressionVertexColor_0.A) {
   code="return float3( VertexOpacity, VertexOpacity, VertexOpacity );"
   inputs [{
     inputName=VertexOpacity
     input {expression=&self:MaterialExpressionVertexColor_0, mask=1, maskA=1, outputIndex=4}
    }]
  }
 ]
}
