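# Unlit engine debug material for a vertex color view mode that isolates the red channel.
# A Vertex Color node feeds a Custom HLSL node that returns red and zeroes green and blue.
# Usage flags cover skeletal, morph target, instanced static, spline and deformed meshes, plus geometry collections.
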Material /Engine/EngineDebugMaterials/VertexColorViewMode_RedOnly
graph {
 properties {
  bUsedWithGeometryCollections=true
  bUsedWithInstancedStaticMeshes=true
  bUsedWithMeshDeformer=true
  bUsedWithMorphTargets=true
  bUsedWithSkeletalMesh=true
  bUsedWithSplineMeshes=true
  pixelDepthOffsetMode=PDOM_Legacy
  shadingModel=MSM_Unlit
  usedShadingModels=Unlit
 }
 nodes [
  MaterialExpressionVertexColor_0 = VertexColor()
  Custom!(VertexColor=@MaterialExpressionVertexColor_0) {
   code="return float3( VertexColor.r, 0.0f, 0.0f );"
   inputs [{
     inputName=VertexColor
     input {expression=&self:MaterialExpressionVertexColor_0, mask=1, maskB=1, maskG=1, maskR=1}
    }]
  }
 ]
}
