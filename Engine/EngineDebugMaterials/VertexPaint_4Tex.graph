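# VertexPaint_4Tex is a lit engine debug material for vertex painting with four diffuse and four normal texture parameters.
# Each texture is multiplied by vertex color and summed within its set, with a tint multiplier on the combined normal.
# Summed diffuse and scaled specular feed ConvertFromDiffSpec, while PowerToRoughness converts a specular power parameter to roughness.
# A separate emissive texture is scaled by a multiplier parameter.
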Material /Engine/EngineDebugMaterials/VertexPaint_4Tex
graph {
 properties {
  bUsedWithInstancedStaticMeshes=true
  bUsedWithMeshDeformer=true
  bUsedWithSkeletalMesh=true
  bUsedWithStaticLighting=true
  pixelDepthOffsetMode=PDOM_Legacy
  usedShadingModels=Default Lit
 }
 nodes [
  MaterialFunctionCall!(Power=ScalarParameter() {defaultValue=15, parameterName=SpecPower}) {materialFunction=/Engine/Functions/Engine_MaterialFunctions01/Shading/PowerToRoughness}
  n1 = VertexColor()
  n2 = Multiply(A=TextureSampleParameter2D() {parameterName=Diffuse, texture=/Engine/Engine_MI_Shaders/T_Reflection}.RGB, B=@n1)
  n3 = Multiply(A=TextureSampleParameter2D() {parameterName=Diffuse2, texture=&T_Reflection}.RGB, B=@n1)
  n4 = Multiply(A=TextureSampleParameter2D() {parameterName=Diffuse3, texture=&T_Reflection}.RGB, B=@n1)
  n5 = Multiply(A=TextureSampleParameter2D() {parameterName=Diffuse4, texture=&T_Reflection}.RGB, B=@n1)
  n6 = TextureSampleParameter2D() {parameterName=Spec, texture=&T_Reflection}
  n7 = ScalarParameter() {defaultValue=1, parameterName=SpecularMultValue}
  MaterialFunctionCall(
   DiffuseColor=Add(A=Add(A=@n2, B=@n3), B=Add(A=@n4, B=@n5))
   SpecularColor=Multiply(A=Multiply(A=@n6.RGB, B=@n7)) {constA=1, constB=0.04, desc=FresnelBaseReflectFraction}
  ) {materialFunction=&ConvertFromDiffSpec}
  n8 = VertexColor()
  n9 = Multiply(A=TextureSampleParameter2D() {parameterName=Normal, samplerType=SAMPLERTYPE_Normal, texture=/Engine/Engine_MI_Shaders/T_Normal}.RGB, B=@n8)
  n10 = Multiply(A=TextureSampleParameter2D() {parameterName=Normal2, samplerType=SAMPLERTYPE_Normal, texture=&T_Normal}.RGB, B=@n8)
  n11 = Multiply(A=TextureSampleParameter2D() {parameterName=Normal3, samplerType=SAMPLERTYPE_Normal, texture=&T_Normal}.RGB, B=@n8)
  n12 = Multiply(A=TextureSampleParameter2D() {parameterName=Normal4, samplerType=SAMPLERTYPE_Normal, texture=&T_Normal}.RGB, B=@n8)
  Multiply(A=Add(A=Add(A=@n9, B=@n10), B=Add(A=@n11, B=@n12)), B=VectorParameter() {parameterName=NormalMultiply, defaultValue=<rgba 1 1 1 1>}.RGB)
  Multiply(
   A=TextureSampleParameter2D() {parameterName=Emissive, texture=/Engine/Engine_MI_Shaders/T_Emissive}.RGB
   B=ScalarParameter() {defaultValue=10, parameterName=EmissiveMultiply}
  )
 ]
}
