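# Engine debug material that renders geometry as an unlit wireframe.
# The material's wireframe flag draws triangle edges, and the Unlit shading model keeps them independent of scene lighting.
# Its only node is a vector parameter named Color, white by default, which material instances can override.
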Material /Engine/EngineDebugMaterials/WireframeMaterial
graph {
 properties {pixelDepthOffsetMode=PDOM_Legacy, shadingModel=MSM_Unlit, usedShadingModels=Unlit, wireframe=true}
 nodes [VectorParameter() {parameterName=Color, defaultValue=<rgba 1 1 1 1>}]
}
